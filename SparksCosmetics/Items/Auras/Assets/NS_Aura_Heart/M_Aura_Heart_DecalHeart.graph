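# Translucent unlit material for Niagara sprite and mesh particles in the Heart aura cosmetic, drawing a glowing heart from a banner icon texture.
# Two differently scaled heart samples, tinted by inner, outer and particle colors, combine with Fresnel and a non-mobile vertical ripple pulse for emissive.
# An opacity chain mixes screen-space scanlines and near-camera fade, then masks by custom depth, or a distance-based backface fade on mobile.
# Static-noise and line-bend branches feed nothing visible.
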Material /SparksCosmetics/Items/Auras/Assets/NS_Aura_Heart/M_Aura_Heart_DecalHeart
graph {
 properties {
  bUsedWithNiagaraMeshParticles=true
  bUsedWithNiagaraSprites=true
  blendMode=BLEND_Translucent
  pixelDepthOffsetMode=PDOM_Legacy
  shadingModel=MSM_Unlit
  usedShadingModels=Unlit
 }
 nodes [
  n1 = Fresnel() {baseReflectFraction=0, exponent=3}
  n2 = VertexNormalWS()
  n3 = CameraVectorWS()
  n4 = Multiply(
   A=DotProduct(A=ComponentMask(None=@n2) {b=true}, B=ComponentMask(None=@n3) {b=true})
   B=ScalarParameter() {defaultValue=0.5, parameterName=LineBend}
  )
  n5 = Constant()
  n6 = ComponentMask(None=ScreenPosition().ViewportUV) {g=true}
  Add(A=QualitySwitch(Default=Multiply(A=@n1, B=@n4) {constB=0.2}, Low=@n5, Medium=@n5), B=@n6)
  n7 = Time() {period=1}
  Clamp(None=@n7)
  n8 = Multiply(A=ScreenPosition().ViewportUV) {constB=3}
  n9 = Time()
  n10 = Panner(Coordinate=TextureCoordinate() {uTiling=2, vTiling=2}, Time=Time()) {speedX=1, speedY=1}
  LinearInterpolate(
   A=TextureSample(UVs=Panner(Coordinate=@n8, Time=@n9) {speedX=0.1, speedY=0.5}) {texture=/Game/Effects/Fort_Effects/Textures/Tile/T_ScanLine_Analog}.RGB
   B=Multiply(A=TextureSample(UVs=@n10) {texture=/Game/Global/Textures/Default/Noise/Noise}.RGB) {constB=20}
   Alpha=ScalarParameter() {parameterName=Static}
  )
  n11 = ScalarParameter() {defaultValue=1, parameterName=Fresnell}
  n12 = TextureCoordinate()
  n13 = ComponentMask(None=@n12) {r=true}
  n14 = Constant() {r=1}
  n15 = ComponentMask(None=@n12) {g=true}
  n16 = Constant() {r=1.35}
  n17 = AppendVector(A=Add(A=Multiply(A=@n13, B=@n14), B=Constant()), B=Add(A=Multiply(A=@n15, B=@n16), B=Constant() {r=0.05}))
  n18 = Constant2Vector() {g=1, r=0.7}
  n19 = Add(
   A=TextureSample(UVs=MaterialFunctionCall(UVs=@n17, Texture Scale=@n18) {materialFunction=&ScaleUVsByCenter}.UVs) {texture=/Game/UI/Foundation/Textures/Banner/Standard/T-Banners-Icons-Heart-L}.R
   B=VectorParameter() {parameterName=Color_Outer, defaultValue=<rgba 0.9725 0.1555 0.3912 1>}.RGB
  )
  n20 = TextureSample(UVs=@n17) {texture=&T-Banners-Icons-Heart-L}
  n21 = VectorParameter() {parameterName=Color_Inner, defaultValue=<rgba 3 0.3624 0.9649 1>}
  n22 = Multiply(A=Multiply(A=@n19, B=Multiply(A=@n20.A, B=@n21.RGB)), B=ParticleColor().RGB)
  n23 = Add(
   A=Multiply(
    A=VectorParameter() {parameterName=FresnelColor, defaultValue=<rgba 1 0.2865 0.5981 1>}.RGB
    B=Fresnel(ExponentIn=@n11) {baseReflectFraction=0}
   )
   B=Multiply(A=VectorParameter() {parameterName=EmissiveColor, defaultValue=<rgba 1 0 0.3302 1>}.RGB, B=Multiply(A=@n19, B=@n22))
  )
  n24 = ActorPositionWS()
  n25 = Multiply(A=ObjectRadius()) {constB=2}
  n26 = Multiply(A=@n25) {constB=-1}
  n27 = Frac(None=Multiply(A=@n7, B=ScalarParameter() {defaultValue=0.5429, parameterName=SpeedRipple}))
  n28 = SphereMask(
   A=ComponentMask(None=WorldPosition().XYZ) {b=true}
   B=Add(A=ComponentMask(None=@n24) {b=true}, B=LinearInterpolate(A=@n26, B=@n25, Alpha=@n27) {constB=0})
   Radius=Multiply(
    A=MaterialFunctionCall() {materialFunction=/Engine/Functions/Engine_MaterialFunctions02/WorldPositionOffset/ObjectScale}.Scale Z
    B=ScalarParameter() {defaultValue=30, parameterName=PulseLineRadius}
   )
  ) {hardnessPercent=0}
  n29 = ShadingPathSwitch(
   Default=Multiply(A=VectorParameter() {parameterName=PulseColor, defaultValue=<rgba 0.913 0.3365 1 1>}.RGB, B=Power(Base=@n28))
   Mobile=Constant()
  ) {desc=Migrated from FeatureLevelSwitch}
  MaterialFunctionCall(Emissive=Add(B=Add(A=@n23, B=@n29))) {materialFunction=&MF_FX_EmissiveAdjustments}
  n30 = ScalarParameter() {defaultValue=0.3, parameterName=Opacity}
  n31 = Sine(None=Multiply(A=@n6, B=ScalarParameter() {defaultValue=200, parameterName=ScanlineTile}))
  n32 = ScalarParameter() {defaultValue=8000, parameterName=FadeDistance}
  n33 = MaterialFunctionCall() {materialFunction=&ObjectScale}
  n34 = LinearInterpolate(
   A=Power(Base=ConstantBiasScale(None=@n31))
   Alpha=MaterialFunctionCall(Distance=Multiply(A=@n32, B=@n33.Scale Z)) {materialFunction=&Near_Camera_Fade}.Result
  ) {constA=0.4, constB=0.3}
  n35 = Add(A=PixelDepth().R, B=Multiply(A=@n33.Scale Z, B=ScalarParameter() {defaultValue=-72.8749, parameterName=BiasSorting_Far}))
  n36 = ComponentMask(None=SceneTexture() {sceneTextureId=PPI_CustomDepth}.Color) {r=true}
  n37 = Constant()
  n38 = Constant() {r=1}
  n39 = ScalarParameter() {defaultValue=1, parameterName=OopacityMobile}
  n40 = Add(A=Distance(A=ObjectPositionWS(), B=CameraPositionWS()), B=ScalarParameter() {defaultValue=-759.463, parameterName=MobileTwosidedFadeOffset})
  n41 = ScalarParameter() {defaultValue=100, parameterName=MobileTwosidedFade}
  n42 = LinearInterpolate(A=@n39, B=ScalarParameter() {parameterName=OopacityMobilebackface}, Alpha=Saturate(None=Divide(A=@n40, B=@n41)))
  n43 = Saturate(None=TwoSidedSign())
  Multiply(
   A=Clamp(None=Add(A=@n30, B=@n34) {constB=0.2}) {maxDefault=2}
   B=ShadingPathSwitch(Default=If(A=@n35, B=@n36, "A > B"=@n37, "A == B"=@n37, "A < B"=@n38), Mobile=LinearInterpolate(A=@n42, B=@n39, Alpha=@n43)) {desc=Migrated from FeatureLevelSwitch}
  )
 ]
}
