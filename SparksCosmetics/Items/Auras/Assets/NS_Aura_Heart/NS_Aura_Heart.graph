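# Niagara system for the Sparks Heart aura cosmetic, attaching a pulsing heart mesh and pink-purple effects to a character's skeletal mesh.
# Heart beats on a looping lifetime curve, bypassing the audio-beat branch, amid short-lived GPU digital-bit sprites.
# Glow sprites pulse on several bones with time-of-day color, while Stars drift from bones and fade; Rays is disabled.
# User.IsLargeBody switches sizes and offsets across emitters, which partition particles to pick bones.
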NiagaraSystem /SparksCosmetics/Items/Auras/Assets/NS_Aura_Heart/NS_Aura_Heart
dependencies {
 usedDataInterfaces [
  &NiagaraDataInterfaceArrayFloat2
  &NiagaraDataInterfaceArrayFloat
  &NiagaraDataInterfaceCurve
  &NiagaraDataInterfaceVector2DCurve
  &NiagaraDataInterfaceArrayInt32
  &NiagaraDataInterfaceColorCurve
  &NiagaraDataInterfaceVectorCurve
  &NiagaraDataInterfaceSkeletalMesh
 ]
 usedDynamicInputs [&MakeVector]
 usedModules [
  &self:SystemSpawnScript.SystemScriptSource.SystemScriptGraph.NiagaraNodeAssignment_0.SetVariables_#1
  &self:SystemSpawnScript.SystemScriptSource.SystemScriptGraph.NiagaraNodeAssignment_1.SetVariables_#2
  &SystemState
 ]
}
emitters {
 Digital_Bits001 {
  simTarget=GPUComputeSim
  emitterUpdateScript [
   &EmitterState()
   &SpawnRate(Spawn Probability=0.7, SpawnRate=50, Use Spawn Probability=false) {module=SpawnRate001}
  ]
  particleSpawnScript [
   &PartitionParticles(Number Of Partitions=7, Partition Mode:ENiagara_PartitionMode=::NewEnumerator5)
   /SetVariables(Particles.ChromaOffset=50) {isSetParameters=true}
   &InitializeParticle(
    Color=&RandomRangeLinearColor(
     Minimum=<rgba 2 0 0.1923 1>
     Maximum=&RandomRangeLinearColor(Minimum=<rgba 1.8 0.739 1.5601 1>, Maximum=<rgba 0.2796 0 0.6198 1>)
    )
    Lifetime Max=0.2
    Lifetime Min=0.05
    Lifetime Mode=::NewEnumerator1
    Sprite Rotation Angle=90
    Sprite Rotation Mode=::NewEnumerator0
    Sprite Size=&MakeCustomVector2DFromBool(
     Bool=$User.IsLargeBody
     True Vector2D=&RandomRangeVector2D(Minimum=<0.2 0.2>, Maximum=<0.6 0.45>)
     False Vector2D=&RandomRangeVector2D(Minimum=<0.1 0.1>, Maximum=<0.5 0.35>)
    )
    Sprite Size Mode=::NewEnumerator4
    Sprite UV Mode=::NewEnumerator3
    UsePositionOffset=false
    Write Material Random=true
   ) {module=InitializeParticle002}
   &CameraOffset(Camera Offset Amount=8)
   &ShapeLocation(
    Apply To Particle Position=::NewEnumerator1
    Box Size=<1 1 1>
    MaskSpawnGroup=false
    Non Uniform Scale=&MakeCustomVectorFromBool(Bool=$User.IsLargeBody, True Vector=<6 5 5>, False Vector=<10 5 5>)
    Offset Coordinate Space=::Local
    Offset Mode=::NewEnumerator0
    Override Seed=false
    Shape Origin=&VectorFromFloat(Value=0)
    Shape Primitive=::NewEnumerator2
    Surface Only=false
    UseOwnerScale=true
   )
  ]
  particleUpdateScript [
   &ParticleState()
   &SkeletalMeshLocation(
    Bone / Socket Index=&SelectIntFromArray(Direct Array Index=$Particles.PartitionParticles.Partition, Int Selection Array=DataInterface() {propertyValues {IntData [7]}})
    Bone Sampling Mode=::NewEnumerator4
    MaskSpawnGroup=false
    Orientation Sampling=::NewEnumerator1
    Position Sampling=::NewEnumerator1
    Sampled Position Offset=&AddVectorToPosition(
     Position=$Particles.ShapeLocation.ShapePosition
     Vector=&MakeCustomVectorFromBool(Bool=$User.IsLargeBody, True Vector=<5 20 -7>, False Vector=<5 20 -7>)
    )
    Skeletal Mesh=$System.Skeletal Mesh
    Source Movement Scale:Vector3f=<0 0 0>
   ) {module=SkeletalMeshLocation002}
   &SolveForcesAndVelocity(Clamp Velocity=false, Limit Acceleration=false)
   /SetVariables(Particles.ChromaOffset:NiagaraFloat=&FloatFromCurve(
     FloatCurve=DataInterface() {propertyValues {Curve {keys [{interpMode=RCIM_Cubic, tangentMode=RCTM_User, time=0, value=1}, {interpMode=RCIM_Cubic, tangentMode=RCTM_User, time=1, value=0}]}}}
     Scale Curve=0.5
    )) {isSetParameters=true}
   &ScaleSpriteSize(Scale Factor=&Vector2DFromCurve(
     Vector2Curve=DataInterface() {propertyValues {
       XCurve {keys [
         {interpMode=RCIM_Cubic, tangentMode=RCTM_User, time=0, value=1}
         {arriveTangent=-6.968, interpMode=RCIM_Cubic, leaveTangent=-6.968, tangentMode=RCTM_User, time=0.5, value=2}
         {tangentMode=RCTM_User, time=1, value=0}
        ]}
       YCurve {keys [
         {interpMode=RCIM_Cubic, tangentMode=RCTM_User, time=0, value=0}
         {arriveTangent=13.8249, interpMode=RCIM_Cubic, leaveTangent=13.8249, tangentMode=RCTM_User, time=0.6, value=2}
         {arriveTangent=8.3745, interpMode=RCIM_Cubic, leaveTangent=8.3745, tangentMode=RCTM_User, time=1, value=8}
        ]}
      }}
     ScaleCurve=<2 8>
    ))
   /SetVariables(Particles.SpriteSize001:Vector2f=&Multiply_Vector2D_ByFloat(Float=1, Vector2D=$Particles.SpriteSize)) {isSetParameters=true}
  ]
  rendererClasses [&NiagaraSpriteRendererProperties]
  renderers [{rendererClass=&NiagaraSpriteRendererProperties}]
 }
 Glow {
  emitterUpdateScript [
   &EmitterState(
    Inactive Response=::NewEnumerator1
    Life Cycle Mode=::NewEnumerator1
    Loop Behavior=::NewEnumerator1
    Loop Duration Mode=::NewEnumerator1
    UseLoopDelay=false
   )
   &SpawnBurst_Instantaneous(Spawn Count=7, Use Spawn Probability=false, UseLoopCountLimit=false)
  ]
  particleSpawnScript [
   &PartitionParticles(Number Of Partitions=7)
   &InitializeParticle(
    Color=<rgba 10 1.9792 5.3383 1>
    Lifetime=0.1
    Sprite Rotation Mode=::NewEnumerator0
    Sprite Size Mode=::NewEnumerator3
    Uniform Sprite Size=&Multiply_Float(
     A=&SelectFloatFromArray(
      Direct Array Index=&SpawnGroup()
      Float Selection Array=DataInterface() {propertyValues {FloatData [40,40,20,20,30,30]}}
     )
     B=&MakeCustomFloatFromBool(Bool=$User.IsLargeBody, True Float=1.5, False Float=1)
    )
    UsePositionOffset=false
    Write Material Random=true
   ) {module=InitializeParticle001}
   &CameraOffset(Camera Offset Amount=5)
   /SetVariables(Particles.glowScale:Vector2f=&Vector2DFromFloat(Value=&Multiply_Float(
      A=&SelectFloatFromArray(
       Direct Array Index=$Particles.PartitionParticles.Partition
       Float Selection Array=DataInterface() {propertyValues {FloatData [150,100,100,90,90,50,50]}}
      )
      B=&MakeCustomFloatFromBool(Bool=$User.IsLargeBody, True Float=1.2, False Float=1)
     ))) {isSetParameters=true}
   /SetVariables(Particles.glowCameraOffset=-10) {isSetParameters=true}
  ]
  particleUpdateScript [
   &ParticleState(Kill Particles When Lifetime Has Elapsed=false, Let Infinitely Lived Particles Die When Emitter Deactivates=true)
   /SetVariables(Particles.glowColor:LinearColor=&FNTimeOfDayColor(
     Night Color=<rgba 3 0.0865 1.1895 0.3>
     Morning Color=<rgba 0.3594 0.146 0.3342 0.4>
     Day Color=<rgba 0.3963 0.1396 0.4323 0.35>
     Evening Color=<rgba 0.3594 0.146 0.3342 0.4>
    )) {isSetParameters=true}
   &ScaleSpriteSize(
    Scale Sprite Size Mode=::NewEnumerator0
    Uniform Scale Factor=&Waveform(
     Global Phase=&Add_Float(A=$Emitter.Age, B=&RandomRangeFloat(Minimum=0, Maximum=5))
     "Waveform [1]"=::NewEnumerator0
     "[1] Frequency"=1
     "[1] Amplitude Min/Max"=<0.85 1>
    )
   )
   &SpriteRotationRate(Rotation Rate=&RandomRangeFloat(Minimum=10, Maximum=20))
   &SkeletalMeshLocation(
    Bone / Socket Index=&SelectIntFromArray(Direct Array Index=$Particles.PartitionParticles.Partition, Int Selection Array=DataInterface() {propertyValues {IntData [0,1,2,3,4,5,6]}})
    Bone Sampling Mode=::NewEnumerator4
    MaskSpawnGroup=false
    Orientation Sampling=::NewEnumerator1
    Sampled Position Offset Coordinate Space:ENiagaraSkelMeshTransforms=::NewEnumerator1
    Skeletal Mesh=$System.Skeletal Mesh
   ) {module=SkeletalMeshLocation003}
   &ScaleColor(
    Scale RGB=&VectorFromFloat(Value=&SelectFloatFromArray(
      Direct Array Index=&SpawnGroup()
      Float Selection Array=DataInterface() {propertyValues {FloatData [1,1,1,1,5,5]}}
     ))
    ScaleA=true
    ScaleRGB=true
   )
   &DynamicMaterialParameters(
    Index 0 Param 1=&SelectFloatFromArray(Direct Array Index=$Particles.PartitionParticles.Partition, Float Selection Array=DataInterface() {propertyValues {FloatData [5,5,5,5,8,8]}})
    Index 0 Param 3=1
    Module.Param0WriteEnabled=true
    Module.Param1WriteEnabled=true
    Module.Param2WriteEnabled=true
    Module.Param3WriteEnabled=true
   )
  ]
  rendererClasses [&NiagaraSpriteRendererProperties]
  renderers [
   {rendererClass=&NiagaraSpriteRendererProperties}
   {rendererClass=&NiagaraSpriteRendererProperties, rendererIndex=1}
  ]
 }
 Heart {
  emitterUpdateScript [
   &EmitterState(
    Inactive Response=::NewEnumerator1
    Life Cycle Mode=::NewEnumerator1
    Loop Behavior=::NewEnumerator1
    Loop Duration Mode=::NewEnumerator1
    UseLoopDelay=false
   )
   &SpawnBurst_Instantaneous(Use Spawn Probability=false, UseLoopCountLimit=false)
  ]
  particleSpawnScript [
   &PartitionParticles(Number Of Partitions=1, Partition Mode:ENiagara_PartitionMode=::NewEnumerator5)
   &InitializeParticle(
    Color=<rgba 0.3375 0.1599 1 1>
    Lifetime=0.5
    Mesh Scale Mode=::NewEnumerator3
    Mesh Uniform Scale=&MakeCustomFloatFromBool(Bool=$User.IsLargeBody, True Float=0.04, False Float=0.03)
    UsePositionOffset=false
    Write Material Random=true
   ) {module=InitializeParticle001}
   &InitialMeshOrientation(Mesh Orientation Mode=::NewEnumerator2, Orientation Vector=<0 0 1>, Rotation=<0 0.09 -0.08>, Use Rotation Vector=true) {enabled=false}
   &InitialMeshOrientation(Mesh Orientation Mode=::NewEnumerator2, Orientation Vector=<0 0 1>, Rotation=<-0.1 0.325 -0.08>, Use Rotation Vector=true) {module=InitialMeshOrientation001}
  ]
  particleUpdateScript [
   &ParticleState(
    Kill Particles When Lifetime Has Elapsed=false
    Let Infinitely Lived Particles Die When Emitter Deactivates=true
    Loop Particles Lifetime=true
   )
   &ScaleMeshSize(Scale Factor=&VectorFromCurve(
     VectorCurve=DataInterface() {propertyValues {
       XCurve {keys [<at 0 1>,<at 0.2 1>,<at 0.3 0.5>,<at 0.5 1>,<at 0.6 1>,<at 0.7 0.5>,<at 0.8 1>,<at 0.9 0.5>,<at 1 1>]}
       YCurve {keys [<at 0 1>,<at 0.2 1>,<at 0.3 0.5>,<at 0.5 1>,<at 0.6 1>,<at 0.7 0.5>,<at 0.8 1>,<at 0.9 0.5>,<at 1 1>]}
       ZCurve {keys [<at 0 1>,<at 0.2 1>,<at 0.3 0.5>,<at 0.5 1>,<at 0.6 1>,<at 0.7 0.5>,<at 0.8 1>,<at 0.9 0.5>,<at 1 1>]}
      }}
     Scale Curve=&MakeCustomVectorFromBool(Bool=$User.IsLargeBody, True Vector=<1 1 1>, False Vector=<1 1 1>)
    )) {enabled=false}
   &ScaleMeshSize(Scale Factor=&VectorFromFloat(Value=&MakeCustomFloatFromBool(
      Bool=true
      True Float=&FloatFromCurve(FloatCurve=DataInterface() {propertyValues {Curve {keys [
           {arriveTangent=4.3642, interpMode=RCIM_Cubic, leaveTangent=4.3642, tangentMode=RCTM_User, time=0, value=0.8}
           {arriveTangent=-0.2186, interpMode=RCIM_Cubic, leaveTangent=-0.2186, tangentMode=RCTM_User, time=0.05, value=1.2}
           {arriveTangent=-0.0405, interpMode=RCIM_Cubic, leaveTangent=-0.0405, tangentMode=RCTM_User, time=1, value=0.8}
          ]}}})
      False Float=&FloatFromCurve(
       FloatCurve=DataInterface() {propertyValues {Curve {keys [
           {arriveTangent=4.3642, interpMode=RCIM_Cubic, leaveTangent=4.3642, tangentMode=RCTM_User, time=0, value=0.8}
           {arriveTangent=-0.2186, interpMode=RCIM_Cubic, leaveTangent=-0.2186, tangentMode=RCTM_User, time=0.05, value=1.2}
           {arriveTangent=-0.0405, interpMode=RCIM_Cubic, leaveTangent=-0.0405, tangentMode=RCTM_User, time=1, value=0.8}
          ]}}}
       CurveIndex=$NPC.NPC_Sparks_Audio_Parameters001.BeatLifetime
      )
     ))) {module=ScaleMeshSize001}
   &SkeletalMeshLocation(
    Bone / Socket Index=&SelectIntFromArray(Direct Array Index=$Particles.PartitionParticles.Partition, Int Selection Array=DataInterface() {propertyValues {IntData [7]}})
    Bone Sampling Mode=::NewEnumerator4
    MaskSpawnGroup=false
    Sampled Position Offset=&MakeCustomVectorFromBool(Bool=$User.IsLargeBody, True Vector=<5 20 -10>, False Vector=<4 20 -8>)
    Skeletal Mesh=$System.Skeletal Mesh
   )
  ]
  rendererClasses [&NiagaraMeshRendererProperties]
  renderers [{rendererClass=&NiagaraMeshRendererProperties, rendererIndex=0}]
 }
 Rays {
  bEnabled=false
  emitterUpdateScript [
   &EmitterState(
    Inactive Response=::NewEnumerator1
    Life Cycle Mode=::NewEnumerator1
    Loop Behavior=::NewEnumerator1
    Loop Duration Mode=::NewEnumerator1
    UseLoopDelay=false
   )
   &SpawnBurst_Instantaneous(Spawn Count=10, Use Spawn Probability=false, UseLoopCountLimit=false)
  ]
  particleSpawnScript [
   &PartitionParticles(Number Of Partitions=10)
   &InitializeParticle(
    Color=&FNTimeOfDayColor(
     Night Color=<rgba 0.4134 0 0.8125 1>
     Morning Color=<rgba 0.4811 0 1 0.1>
     Day Color=<rgba 0.2305 0 1 0.1>
     Evening Color=<rgba 0.3309 0 1 0.1>
    )
    Lifetime=&RandomRangeFloat(Minimum=4, Maximum=2)
    Sprite Rotation Angle=&SelectFloatFromArray(
     Direct Array Index=$Particles.PartitionParticles.Partition
     Float Selection Array=DataInterface() {propertyValues {FloatData [180,0,90,-90,45,-45,140,-136,40,-145]}}
    )
    Sprite Rotation Mode=::NewEnumerator0
    Sprite Size=&Multiply_Vector2D(
     A=&SelectVector2DFromArray(
      Direct Array Index=$Particles.PartitionParticles.Partition
      Vector 2D Selection Array=DataInterface() {propertyValues {FloatData [<50 60>, <40 60>, <30 40>, <30 50>, <30 60>, <10 50>, <10 60>, <10 60>, <20 40>, <20 60>]}}
     )
     B=&MakeCustomVector2DFromBool(Bool=$User.IsLargeBody, True Vector2D=<1.5 1.5>, False Vector2D=<1 1>)
    )
    Sprite Size Mode=::NewEnumerator4
    UsePositionOffset=false
    Write Material Random=true
   )
   &CameraOffset(Camera Offset Amount=-20)
  ]
  particleUpdateScript [
   &ParticleState(
    Kill Particles When Lifetime Has Elapsed=false
    Let Infinitely Lived Particles Die When Emitter Deactivates=true
    Loop Particles Lifetime=true
   )
   &ScaleSpriteSize(
    Non-Uniform Curve Scale=&RandomRangeVector2D(Minimum=<1 1>, Maximum=<1.1 1.1>)
    Non-Uniform Curve Sprite Scale=DataInterface() {propertyValues {
      XCurve {keys [
        {arriveTangent=-0.164, interpMode=RCIM_Cubic, time=0, value=0.8}
        <at 0.5 1 RCIM_Cubic>
        {interpMode=RCIM_Cubic, leaveTangent=-0.7766, time=1, value=0.8}
       ]}
      YCurve {keys [<at 0 0.6 RCIM_Cubic>,<at 0.5 1 RCIM_Cubic>,<at 1 0.6 RCIM_Cubic>]}
     }}
    Scale Sprite Size Mode=::NewEnumerator3
   )
   &SpriteRotationRate(Rotation Rate=20)
   &SkeletalMeshLocation(
    Bone / Socket Index=7
    Bone Sampling Mode=::NewEnumerator4
    MaskSpawnGroup=false
    Orientation Sampling=::NewEnumerator1
    Skeletal Mesh=$System.Skeletal Mesh
   )
   &SolveForcesAndVelocity(Clamp Velocity=false, Limit Acceleration=false)
   &ScaleColor(
    Scale Alpha=&FloatFromCurve(
     FloatCurve=DataInterface() {propertyValues {Curve {keys [
         {interpMode=RCIM_Cubic, tangentMode=RCTM_User, time=0, value=0.2}
         {interpMode=RCIM_Cubic, tangentMode=RCTM_User, time=0.5, value=1}
         {tangentMode=RCTM_User, time=1, value=0.2}
        ]}}}
     Scale Curve=&RandomRangeFloat(Minimum=0.8, Maximum=1)
    )
    ScaleA=true
    ScaleRGB=true
   )
  ]
  rendererClasses [&NiagaraSpriteRendererProperties]
  renderers [{rendererClass=&NiagaraSpriteRendererProperties}]
 }
 Stars {
  emitterUpdateScript [
   &EmitterState(
    Inactive Response=::NewEnumerator1
    Life Cycle Mode=::NewEnumerator1
    Loop Behavior=::NewEnumerator1
    Loop Duration Mode=::NewEnumerator1
    UseLoopDelay=false
   )
   &SpawnRate(SpawnRate=10, Use Spawn Probability=false)
  ]
  particleSpawnScript [
   &PartitionParticles(Number Of Partitions=5)
   &InitializeParticle(
    Color=<rgba 5 0 0.1864 1>
    Lifetime=&RandomRangeFloat(Minimum=1.2, Maximum=1)
    Sprite Size Mode=::NewEnumerator3
    Uniform Sprite Size=&MakeCustomFloatFromBool(
     Bool=$User.IsLargeBody
     True Float=&RandomRangeFloat(Minimum=45, Maximum=25)
     False Float=&RandomRangeFloat(Minimum=35, Maximum=20)
    )
    UsePositionOffset=false
    Write Material Random=true
   )
   &AddVelocity(
    Velocity=&RandomRangeVector(Minimum=<0 20 100>, Maximum=<0 -20 50>)
    Velocity Speed Scale=&MakeCustomFloatFromBool(Bool=$User.IsLargeBody, True Float=1.5, False Float=1.4)
   )
   &CameraOffset(Camera Offset Amount=-30)
  ]
  particleUpdateScript [
   &ParticleState()
   &ScaleSpriteSize(Uniform Curve Sprite Scale=DataInterface() {propertyValues {Curve {keys [
        {tangentMode=RCTM_User, time=0, value=0}
        {arriveTangent=1.4286, interpMode=RCIM_Cubic, leaveTangent=-0.3443, tangentMode=RCTM_Break, time=0.15, value=1.2}
        {arriveTangent=-0.4706, interpMode=RCIM_Cubic, leaveTangent=-0.4706, time=0.7, value=1}
        <at 1 0.8>
       ]}}})
   &SkeletalMeshLocation(
    Bone / Socket Index=&SelectIntFromArray(Direct Array Index=$Particles.PartitionParticles.Partition, Int Selection Array=DataInterface() {propertyValues {IntData [0,1,2,3,4]}})
    Bone Sampling Mode=::NewEnumerator4
    MaskSpawnGroup=false
    Position Sampling=::NewEnumerator1
    Sampled Position Offset=<-10 0 0>
    Skeletal Mesh=$System.Skeletal Mesh
    Source Movement Scale:Vector3f=<0 0 0>
   ) {module=SkeletalMeshLocation002}
   &Drag(
    Drag=&FloatFromCurve(
     FloatCurve=DataInterface() {propertyValues {Curve {keys [{arriveTangent=2.934, interpMode=RCIM_Cubic, leaveTangent=2.934, tangentMode=RCTM_User, time=0, value=0}, <at 1 1 RCIM_Cubic>]}}}
     Scale Curve=5
    )
    Rotational Drag=4
    Use Linear Drag=true
    Use Rotational Drag=false
   )
   &SolveForcesAndVelocity(Clamp Velocity=false, Limit Acceleration=false)
   &ScaleColor(
    Scale Alpha=&FloatFromCurve(FloatCurve=DataInterface() {propertyValues {Curve {keys [{interpMode=RCIM_Cubic, tangentMode=RCTM_User, time=0, value=1}, {interpMode=RCIM_Cubic, tangentMode=RCTM_User, time=1, value=0}]}}})
    Scale RGB=&MakeVectorFromLinearColorRGB(Color=&ColorFromCurve(ColorCurve=DataInterface() {propertyValues {
        AlphaCurve {keys [<at 0 1 RCIM_Cubic>,<at 1 0 RCIM_Cubic>]}
        BlueCurve {keys [
          <at 0 5>
          {arriveTangent=-16.7575, interpMode=RCIM_Cubic, leaveTangent=-16.7575, time=0.0717, value=4}
          {interpMode=RCIM_Cubic, leaveTangent=-5.8609, time=0.179, value=2}
          <at 0.4766 1>
         ]}
        GreenCurve {keys [
          <at 0 5>
          {arriveTangent=-16.7575, interpMode=RCIM_Cubic, leaveTangent=-16.7575, time=0.0717, value=4}
          {interpMode=RCIM_Cubic, leaveTangent=-7.8145, time=0.179, value=2}
          <at 0.4766 0>
         ]}
        RedCurve {keys [
          <at 0 5>
          {arriveTangent=-16.7575, interpMode=RCIM_Cubic, leaveTangent=-16.7575, time=0.0717, value=4}
          {interpMode=RCIM_Cubic, leaveTangent=-7.5447, time=0.179, value=2}
          <at 0.4766 0.1381>
         ]}
       }}))
    ScaleA=true
    ScaleRGB=true
   )
  ]
  rendererClasses [&NiagaraSpriteRendererProperties]
  renderers [{rendererClass=&NiagaraSpriteRendererProperties}]
 }
}
summary {
 systemName=NS_Aura_Heart
 emitters [
  {bEnabled=false, emitterName=Rays, rendererClasses [&NiagaraSpriteRendererProperties]}
  {emitterName=Stars, rendererClasses [&NiagaraSpriteRendererProperties]}
  {emitterName=Glow, rendererClasses [&NiagaraSpriteRendererProperties]}
  {emitterName=Digital_Bits001, simTarget=GPUComputeSim, rendererClasses [&NiagaraSpriteRendererProperties]}
  {emitterName=Heart, rendererClasses [&NiagaraMeshRendererProperties]}
 ]
 userVariables [
  User.IsFrontEnd {type=NiagaraBool, defaultValue=true}
  User.IsLargeBody {type=NiagaraBool}
 ]
}
